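# Unlit translucent Niagara flare material for the BiruFang pickaxe FX.
# Sphere-mask core and edge glows plus a flare texture take particle color or animated rainbow, with optional hue jitter.
# Multiplied sines create flicker, driven by dynamic parameters or per-particle random for GPU particles, combined with depth, view-angle and near-camera fades.
# An optional distance switch computes a vertex offset that enlarges the flare far from camera.
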Material /BRCosmetics/Weapons/FORT_Melee/Pickaxe_BiruFang/FX/M_BiruFang_AlphaBlend_Flare
graph {
 properties {
  bAllowNegativeEmissiveColor=true
  bUsedWithNanite=true
  bUsedWithNiagaraMeshParticles=true
  bUsedWithNiagaraSprites=true
  blendMode=BLEND_Translucent
  pixelDepthOffsetMode=PDOM_Legacy
  shadingModel=MSM_Unlit
  usedShadingModels=Unlit
 }
 nodes [
  n1 = DynamicParameter() {paramNames [FlickerFade,FlickerSpeed,ColorJitter,Param4]}
  n2 = ScalarParameter() {defaultValue=2, parameterName=CenterBrightness}
  n3 = Constant() {r=0.5}
  n4 = SphereMask(A=TextureCoordinate(), B=@n3, Radius=@n3) {hardnessPercent=0}
  n5 = Power(Base=@n4, Exp=ScalarParameter() {defaultValue=25, parameterName=CentertPower}) {constExponent=25}
  n6 = ComponentMask(None=TextureCoordinate() {uTiling=0, vTiling=0}) {g=true}
  n7 = Multiply(
   A=TextureSample(UVs=Panner(Coordinate=@n6) {speedX=1.6667}) {texture=/Game/Global/Textures/Default/Masks_and_Gradients/T_Grad_Rainbow}.RGB
   B=ParticleColor().RGB
  )
  n8 = ParticleColor()
  n9 = StaticSwitchParameter(
   True=MaterialFunctionCall(LDR Input=Clamp(None=@n7), HDR Tint=@n7) {materialFunction=&DeriveHDRfromLDR}.RemappedRange
   False=Desaturation(None=@n8.RGB, Fraction=ScalarParameter() {parameterName=inputDesaturation})
  ) {parameterName=Use Rainbow Color}
  n10 = ScalarParameter() {defaultValue=0.3, parameterName=CenterDesaturation}
  n11 = Multiply(
   A=Power(Base=@n4, Exp=ScalarParameter() {defaultValue=4, parameterName=EdgePower}) {constExponent=4}
   B=ScalarParameter() {defaultValue=0.15, parameterName=EdgeBrightness}
  ) {constB=0.15}
  n12 = VectorParameter() {defaultValue=<rgba 0 0 0 1>}
  n13 = TextureSampleParameter2D(UVs=Add(A=TextureCoordinate(), B=ComponentMask(None=@n12.RGB) {g=true, r=true})) {parameterName=FlareMask, texture=/Game/Effects/Fort_Effects/Textures/Flares/T_Flare_Spikeball}
  n14 = Add(A=Multiply(A=Multiply(A=@n2, B=@n5) {constB=2}, B=Desaturation(None=@n9, Fraction=@n10)), B=Multiply(A=Add(A=@n11, B=@n13.RGB), B=@n9))
  n15 = StaticSwitchParameter(
   True=MaterialFunctionCall(Hue Shift Percentage=@n1.ColorJitter, Texture=@n14) {materialFunction=&HueShift}.Result
   False=@n14
  ) {parameterName=UseColorJitter?}
  MaterialFunctionCall(Emissive=@n15) {materialFunction=&MF_FX_EmissiveAdjustments}
  n16 = MaterialFunctionCall("FadeSensitivity (optional)"=ScalarParameter() {defaultValue=1, parameterName=AngleFade}) {materialFunction=&FadeBasedOnViewAngle}
  n17 = Constant() {r=1}
  n18 = ParticleRandom()
  n19 = Multiply(
   A=Time()
   B=StaticSwitchParameter(True=LinearInterpolate(Alpha=@n18) {constA=1.2, constB=0.8}, False=@n1.FlickerSpeed) {parameterName=GPU Particles}
  )
  n20 = Sine(None=@n19) {period=0.075}
  n21 = Multiply(A=Multiply(A=Sine(None=@n19) {period=0.155}, B=Sine(None=@n19) {period=0.2}), B=Sine(None=@n19) {period=0.43})
  n22 = ScalarParameter() {defaultValue=0.4, desc="0 = full flicker amount", parameterName=GPUFlickerFadeMin}
  n23 = ScalarParameter() {defaultValue=1, desc="1 = no flicker", parameterName=GPUFlickerFadeMax, sortPriority=1}
  n24 = ParticleRandom()
  n25 = LinearInterpolate(
   A=Clamp(None=Multiply(A=@n20, B=@n21)) {maxDefault=3}
   Alpha=StaticSwitchParameter(True=LinearInterpolate(A=@n22, B=@n23, Alpha=@n24) {constA=0.5}, False=Saturate(None=@n1.FlickerFade)) {parameterName=GPU Particles}
  ) {constA=1}
  n26 = Constant() {r=1}
  n27 = DynamicParameter() {desc=Index 1 Params for Niagara, parameterIndex=1, paramNames [DepthFadeDist, Param2, Param3, Param4]}
  n28 = ScalarParameter() {defaultValue=30, parameterName=DepthFade}
  n29 = DepthFade(Opacity=@n8.A, FadeDistance=If(A=@n27.DepthFadeDist, B=Constant() {r=1}, "A > B"=@n27.DepthFadeDist, "A == B"=@n28, "A < B"=@n28))
  n30 = Multiply(
   A=Multiply(
    A=StaticSwitchParameter(True=@n16.Result, False=@n17) {parameterName=Use Angle Fade}
    B=StaticSwitchParameter(True=@n25, False=@n26) {defaultValue=true, parameterName=UseFlicker}
   )
   B=ShadingPathSwitch(Default=@n29, Deferred=@n29, Mobile=@n8.A) {desc=Migrated from FeatureLevelSwitch}
  )
  n31 = MaterialFunctionCall(Distance=ScalarParameter() {parameterName=CameraFadeDistance}) {materialFunction=&Near_Camera_Fade}
  Multiply(
   A=@n15
   B=StaticSwitchParameter(True=Multiply(A=@n30, B=@n31.Result), False=@n30) {group=DetailedGradientMask, parameterName=UseCameraNearFade}
  )
  ShadingPathSwitch(Default=@n25, Mobile=Constant() {r=0.8}) {desc=Migrated from FeatureLevelSwitch}
  n32 = ScalarParameter() {defaultValue=10, parameterName=MaxScaleIncrease}
  n33 = ObjectPositionWS()
  n34 = Multiply(A=Subtract(A=@n33, B=WorldPosition().XYZ)) {constB=-1}
  n35 = CameraPositionWS()
  n36 = Divide(
   A=Subtract(A=Distance(A=@n33, B=@n35), B=ScalarParameter() {defaultValue=2000, parameterName=DistanceOffset})
   B=ScalarParameter() {defaultValue=30000, parameterName=ScaleUpDistance}
  )
  StaticSwitchParameter(True=LinearInterpolate(B=Multiply(A=@n32, B=@n34) {constB=-1}, Alpha=Clamp(None=@n36)) {constB=0}, False=Constant()) {parameterName=Use Distance Fade}
 ]
}
