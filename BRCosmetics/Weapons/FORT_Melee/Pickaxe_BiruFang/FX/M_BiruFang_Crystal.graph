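# Unlit emissive crystal material for the BiruFang pickaxe effects, usable on skeletal and Niagara particle meshes.
# Panning noise bump-offsets a turbulence texture, and their difference distorts the radius of a sphere mask around the object's position.
# Mask, view-angle falloff, turbulence and a constant full-body glow term drive StormKing skin curve-atlas ramps for color.
# Blurred noise, time-of-day tint and particle color scale the resulting emissive, the only connected output.
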Material /BRCosmetics/Weapons/FORT_Melee/Pickaxe_BiruFang/FX/M_BiruFang_Crystal
graph {
 properties {
  bUseMaterialAttributes=true
  bUsedWithMeshDeformer=true
  bUsedWithNanite=true
  bUsedWithNiagaraMeshParticles=true
  bUsedWithSkeletalMesh=true
  ditherOpacityMask=true
  pixelDepthOffsetMode=PDOM_Legacy
  shadingModel=MSM_Unlit
  usedShadingModels=Unlit
 }
 nodes [
  n1 = MaterialFunctionCall(
   Color_Morning=Constant3Vector() {constant=<rgba 0.25 0.3334 0.5 1>}
   Color_Day=Constant3Vector() {constant=<rgba 0.6 0.6 1 0>}
   Color_Evening=Constant3Vector() {constant=<rgba 0.0383 0.025 0.05 1>}
   Color_Night=Constant3Vector() {constant=<rgba 0.0469 0.0469 0.0781 0>}
  ) {materialFunction=&Color_Time_of_Day}
  n2 = OneMinus(None=TextureSample() {samplerType=SAMPLERTYPE_Grayscale, texture=/FortWater/Textures/LowResBlurredNoise}.G)
  n3 = TextureCoordinate()
  n4 = ScalarParameter() {defaultValue=2, parameterName=UV}
  n5 = Time()
  n6 = TextureSample(UVs=Add(A=Multiply(A=@n3, B=@n4), B=Multiply(A=@n5) {constB=0.02})) {samplerType=SAMPLERTYPE_LinearColor, texture=&T_NoisePack_LinearColor}
  n7 = TextureCoordinate()
  n8 = ScalarParameter() {defaultValue=2, parameterName=UV2}
  n9 = TextureSample(UVs=BumpOffset(Coordinate=Multiply(A=@n7, B=@n8), Height=LinearInterpolate(Alpha=@n6.G) {constA=0.5}, HeightRatioInput=Constant() {r=0.5})) {texture=/Game/Global/Textures/Default/Caustics/T_Turbulence_Seamless}
  n10 = Subtract(A=@n6.R, B=@n9.R)
  n11 = Constant() {r=2000}
  n12 = ScalarParameter() {defaultValue=1, parameterName=WaveRadius}
  n13 = SphereMask(
   A=WorldPosition().XYZ
   B=ComponentMask(None=ObjectPositionWS()) {b=true, g=true, r=true}
   Radius=Multiply(A=LinearInterpolate(Alpha=@n10) {constA=0.8}, B=Multiply(A=@n11, B=@n12))
   Hardness=Constant()
  )
  n14 = CurveAtlasRowParameter(CurveTime=Add(A=@n13) {constB=-1}) {
   atlas=&AthenaCurveAtlas01
   curve=/Game/Athena/LinearColorCurves/Curve_StormKingSkinFXRamp4
   defaultValue=32
   parameterName=RampCurve
  }
  n15 = Constant() {r=1}
  n16 = OneMinus(None=Abs(None=Sine(None=@n15)))
  n17 = Power(Base=@n15, Exp=ScalarParameter() {defaultValue=5, sortPriority=3}) {constExponent=3}
  n18 = ScalarParameter() {defaultValue=2, parameterName=Emissive_Intensity, sortPriority=3}
  n19 = Multiply(
   A=Multiply(A=LinearInterpolate(Alpha=@n16) {constA=0.5}, B=Multiply(A=@n17, B=@n18) {constB=3})
   B=ScalarParameter() {defaultValue=0.5, parameterName=FullBodyGlow}
  )
  n20 = Saturate(None=Add(A=@n13, B=@n19))
  n21 = DotProduct(A=Transform(None=VertexNormalWS()), B=CameraVectorWS())
  n22 = LinearInterpolate(A=Multiply(A=@n20, B=@n20), B=@n20, Alpha=Multiply(A=@n21, B=@n21)) {constA=0.2}
  n23 = LinearInterpolate(Alpha=SquareRoot(None=@n9.R)) {constA=0.4, constB=1.5}
  n24 = CurveAtlasRowParameter(CurveTime=Multiply(A=Multiply(A=@n14.R, B=@n22), B=@n23)) {
   atlas=&AthenaCurveAtlas01
   curve=/Game/Athena/LinearColorCurves/Curve_StormKingSkinFXColor5
   defaultValue=33
   parameterName=ColorCurve
  }
  n25 = CurveAtlasRowParameter(CurveTime=Multiply(A=@n19, B=@n23)) {
   atlas=&AthenaCurveAtlas01
   curve=&Curve_StormKingSkinFXColor5
   defaultValue=33
   parameterName=ColorCurve
  }
  n26 = ScalarParameter() {defaultValue=0.5, parameterName=FullBodyGlow}
  n27 = Multiply(A=LinearInterpolate(A=Multiply(A=@n2, B=@n2)) {constAlpha=0.2, constB=0.2}, B=Add(A=@n24, B=Multiply(A=@n25, B=@n26)))
  MakeMaterialAttributes!(
   BaseColor
   Metallic
   Specular
   Roughness
   Anisotropy
   EmissiveColor=Multiply(A=Multiply(A=@n1.Result, B=@n27), B=ParticleColor().RGB)
   Opacity
   OpacityMask
   Normal
   Tangent
   WorldPositionOffset
   SubsurfaceColor
   ClearCoat
   ClearCoatRoughness
   AmbientOcclusion
   Refraction
   CustomizedUVs_0
   CustomizedUVs_1
   CustomizedUVs_2
   CustomizedUVs_3
   CustomizedUVs_4
   CustomizedUVs_5
   CustomizedUVs_6
   CustomizedUVs_7
   PixelDepthOffset
   ShadingModel
   Displacement
  )
 ]
}
