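# M_BiruFang_Glow is a translucent, unlit, two-sided Niagara material for a smoky glow effect on the BiruFang pickaxe.
# Two panning smoke noise samples, with tiling and speed driven partly by dynamic parameters, are blended and added to a flipped gradient texture.
# Particle color tints the result for emissive.
# Opacity multiplies that pattern by the gradient, then fades it by view angle, depth, inverse Fresnel and particle alpha.
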Material /BRCosmetics/Weapons/FORT_Melee/Pickaxe_BiruFang/FX/M_BiruFang_Glow
graph {
 properties {
  bUsedWithNiagaraMeshParticles=true
  bUsedWithNiagaraSprites=true
  blendMode=BLEND_Translucent
  pixelDepthOffsetMode=PDOM_Legacy
  shadingModel=MSM_Unlit
  twoSided=true
  usedShadingModels=Unlit
 }
 nodes [
  n1 = ParticleColor()
  n2 = DynamicParameter() {paramNames [YTile,PanSpeed,Param3,Param4]}
  n3 = AppendVector(A=ScalarParameter() {defaultValue=1, parameterName=X_Tile}, B=@n2.YTile)
  n4 = MaterialFunctionCall(XY=TextureCoordinate() {uTiling=0.1, vTiling=8}) {materialFunction=&Swizzle}
  n5 = Multiply(A=Time(), B=@n2.PanSpeed)
  n6 = ScalarParameter() {defaultValue=-0.5, parameterName=NoiseA_SpeedX}
  n7 = ScalarParameter() {defaultValue=-1, parameterName=NoiseA_SpeedY}
  n8 = TextureSampleParameter2D(UVs=Panner(Coordinate=Multiply(A=@n3, B=@n4.YX), Time=@n5, Speed=AppendVector(A=@n6, B=@n7)) {speedX=0.2, speedY=1}) {parameterName=NoiseA, texture=&T_Inky_Smoke_Tile_Chromatic}
  n9 = MaterialFunctionCall(XY=TextureCoordinate() {uTiling=0.1, vTiling=4}) {materialFunction=&Swizzle}
  n10 = ScalarParameter() {defaultValue=0.25, parameterName=NoiseB_SpeedX}
  n11 = ScalarParameter() {defaultValue=-1, parameterName=NoiseB_SpeedY}
  n12 = TextureSampleParameter2D(UVs=Panner(Coordinate=Multiply(A=@n3, B=@n9.YX), Time=@n5, Speed=AppendVector(A=@n10, B=@n11)) {speedX=-0.25, speedY=0.8}) {parameterName=NoiseB, texture=&T_Inky_Smoke_Tile_Chromatic}
  n13 = TextureCoordinate()
  n14 = OneMinus(None=ComponentMask(None=@n13) {r=true})
  n15 = TextureSample(UVs=AppendVector(A=Add(A=@n14) {constB=0}, B=ComponentMask(None=@n13) {g=true})) {
   samplerSource=SSM_Clamp_WorldGroupSettings
   texture=/Game/Characters/Player/Female/Medium/Bodies/F_MED_Neon_Cat_Spy/Textures/T_F_MED_NeonCatSpy_Gradients_TrimDetail_Black
  }
  n16 = Power(Base=@n15.RGB) {constExponent=0.5}
  n17 = Add(
   A=Power(Base=LinearInterpolate(A=@n8.RGB, B=@n12.RGB), Exp=ScalarParameter() {defaultValue=0.5, parameterName=Noise_Exponent})
   B=Multiply(A=@n16) {constB=1.5}
  )
  MaterialFunctionCall(Emissive=Multiply(A=@n1.RGB, B=@n17)) {materialFunction=&MF_FX_EmissiveAdjustments}
  n18 = Fresnel() {baseReflectFraction=0.05, exponent=1}
  n19 = Multiply(A=@n17, B=@n16)
  n20 = ScalarParameter() {defaultValue=1.25, parameterName=FadeBasedOnViewAngle}
  n21 = Saturate(None=Multiply(
    A=Power(Base=@n19)
    B=MaterialFunctionCall("FadeSensitivity (optional)"=@n20) {materialFunction=&FadeBasedOnViewAngle}.Result
   ))
  Multiply(A=@n1.A, B=Multiply(A=OneMinus(None=@n18), B=DepthFade(Opacity=@n21) {fadeDistanceDefault=10}))
  TextureCoordinate()
 ]
}
