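# Unlit, translucent, two-sided Niagara ribbon material for the main trail of the BiruFang pickaxe.
# Opacity uses a panning primary mask, blended by TrailFadeOff between a center-peaked U falloff and a square-rooted gradient-mask fade.
# Particle alpha and depth fade then scale that opacity.
# Emissive tints particle color, multiplied by panning smoke noise on UV1 unless UseDefaultColor is on, and passes through MF_FX_EmissiveAdjustments.
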Material /BRCosmetics/Weapons/FORT_Melee/Pickaxe_BiruFang/FX/M_BiruFang_MainTrail
graph {
 properties {
  bAutomaticallySetUsageInEditor=false
  bUsedWithNiagaraRibbons=true
  bUsedWithStaticMesh=false
  blendMode=BLEND_Translucent
  pixelDepthOffsetMode=PDOM_Legacy
  shadingModel=MSM_Unlit
  twoSided=true
  usedShadingModels=Unlit
 }
 nodes [
  n1 = Multiply(
   A=TextureCoordinate()
   B=AppendVector(A=ScalarParameter() {defaultValue=1, parameterName=PrimaryMaskTileU}, B=ScalarParameter() {defaultValue=1, parameterName=PrimaryMaskTileV})
  )
  n2 = AppendVector(A=ScalarParameter() {defaultValue=1, parameterName=PrimaryMaskAddU}, B=ScalarParameter() {defaultValue=1, parameterName=PrimaryMaskAddV})
  n3 = ScalarParameter() {parameterName=PrimaryMaskPanSpeedU}
  n4 = ScalarParameter() {parameterName=PrimaryMaskPanSpeedV}
  n5 = TextureSampleParameter2D(UVs=Panner(Coordinate=Add(A=@n1, B=@n2), Speed=AppendVector(A=@n3, B=@n4)) {desc=Default Speed is 0}) {parameterName=PrimaryMask, texture=/Game/Global/Textures/Default/Masks_and_Gradients/T_Water_Top_Ramp}
  n6 = ComponentMask(None=TextureCoordinate()) {r=true}
  n7 = Multiply(A=@n5.RGB, B=Multiply(A=OneMinus(None=@n6), B=@n6))
  n8 = TextureSampleParameter2D(UVs=MaterialFunctionCall() {materialFunction=&Swizzle}.YX) {parameterName=Mask, texture=/Game/Effects/Fort_Effects/Textures/Masks/T_Vert_Gradient_Tilable}
  n9 = TextureCoordinate()
  n10 = Multiply(A=OneMinus(None=ComponentMask(None=@n9) {g=true}), B=OneMinus(None=ComponentMask(None=@n9) {r=true}))
  n11 = Power(Base=Multiply(A=@n5.RGB, B=Multiply(A=@n8.RGB, B=@n10))) {constExponent=0.5}
  n12 = ScalarParameter() {parameterName="TrailFadeOff(0,1)"}
  n13 = ParticleColor()
  DepthFade(
   Opacity=Multiply(A=LinearInterpolate(A=@n7, B=@n11, Alpha=@n12), B=@n13.A)
   FadeDistance=ScalarParameter() {defaultValue=35, parameterName=DepthFadeDistance}
  )
  n14 = Multiply(
   A=TextureCoordinate() {coordinateIndex=1}
   B=AppendVector(A=ScalarParameter() {defaultValue=0.01, parameterName=CloudTileU}, B=ScalarParameter() {defaultValue=0.2, parameterName=CloudTileV})
  )
  n15 = AppendVector(A=ScalarParameter() {defaultValue=-0.1, parameterName=TilingNoisePanSpeedU}, B=ScalarParameter() {parameterName=TilingNoisePanSpeedV})
  n16 = Multiply(
   A=@n13.RGB
   B=TextureSampleParameter2D(UVs=Panner(Coordinate=@n14, Speed=@n15) {desc=Default Speed is 0}) {parameterName=TilingNoise, texture=/Game/Global/Textures/Default/Tileable_Patterns/T_Smoke_Puff_Tile_Up}.RGB
  )
  MaterialFunctionCall(Emissive=Multiply(
    A=VectorParameter() {parameterName=TrailFXColor, defaultValue=<rgba 1 1 1 1>}.RGB
    B=StaticSwitchParameter(True=@n13.RGB, False=@n16) {parameterName=UseDefaultColor}
   )) {materialFunction=&MF_FX_EmissiveAdjustments}
 ]
}
