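# Master material for the BiruFang pickaxe, built on the shared character PreFX and PostFX functions with Default Lit and Subsurface Profile shading.
# A TechArtMask texture drives a thin-film effect on the PreFX output.
# A step on the mask's red channel limits a pre-skinned Y-position gradient on base color and a crystal function borrowed from the RoyalDusk backpack, scaled by BF_MaskMultiply.
# The result feeds PostFX, then cinematic post FX.
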Material /BRCosmetics/Weapons/FORT_Melee/Pickaxe_BiruFang/Materials/M_BiruFang_Pickaxe
graph {
 properties {
  bAutomaticallySetUsageInEditor=false
  bUseMaterialAttributes=true
  bUsedWithClothing=true
  bUsedWithMeshDeformer=true
  bUsedWithMorphTargets=true
  bUsedWithNanite=true
  bUsedWithSkeletalMesh=true
  pixelDepthOffsetMode=PDOM_Legacy
  refractionMethod=RM_IndexOfRefraction
  shadingModel=MSM_FromMaterialExpression
  subsurfaceProfile=&SS_HeroSkin_02
  usedShadingModels=Default Lit | Subsurface Profile
 }
 nodes [
  n1 = MaterialFunctionCall() {materialFunction=&BaseFN_CharacterFunction_PreFX}
  Composite!("Material (MA)"=@n1.Result) {subgraphName=VFX}
  MaterialExpressionReroute_0 = Reroute(None=@n1.Result)
  n2 = Constant() {r=1}
  n3 = TextureSampleParameter2D() {
   group=05_Tech Art Masks
   parameterName=TechArtMask
   samplerType=SAMPLERTYPE_Masks
   sortPriority=0
   texture=&T_BlackMask
  }
  n4 = GetMaterialAttributes(None=MaterialFunctionCall!(Material Attributes=@MaterialExpressionReroute_0."Material (MA)", Fresnel=Fresnel(ExponentIn=@n2) {baseReflectFraction=0}, Alpha=@n3.RGB) {materialFunction=/Game/Global/MaterialFunctions/TechArt/MF_ThinFilm_TechArt}.Result) {attributeGetTypes [#1]}
  n5 = PreSkinnedPosition()
  n6 = VertexInterpolator(VS=Multiply(A=Subtract(A=@n5) {constB=55}) {constB=0.003})
  n7 = Subtract(A=@n4.Base Color, B=ComponentMask(None=Multiply(A=@n6.PS) {constB=-1}) {g=true})
  n8 = Step(X=@n3.R) {constY=0.1}
  n9 = MaterialFunctionCall!(
   Material=SetMaterialAttributes!(MaterialAttributes=@n4.MaterialAttributes, Base Color=LinearInterpolate(A=@n4.Base Color, B=@n7, Alpha=@n8)) {attributeSetTypes [#1]}
   Alpha=Multiply(A=@n8, B=ScalarParameter() {defaultValue=1, group=BF_Material, parameterName=BF_MaskMultiply})
  ) {materialFunction=/Game/Accessories/FORT_Backpacks/Backpack_F_MED_RoyalDusk/Materials/MF_RoyalDusk_Crystal}
  MaterialExpressionReroute_1 = Reroute(None=@n9.Result)
  n10 = MaterialFunctionCall(
   AO=@n1.AO
   Distance Blend=@n1.AO
   Fresnel=@n1.AO
   Gmap=@n1.AO
   Pre Skinned Local Position=@n1.AO
   Material=@MaterialExpressionReroute_1.Result
  ) {materialFunction=&BaseFN_CharacterFunction_PostFX}
  MaterialFunctionCall(In=@n10.Result) {materialFunction=&CINE_MF_CharPostFX}
  MaterialFunctionCall(Material Attributes=@n9.Result, Alpha=OneMinus(None=@n8)) {materialFunction=&MF_ThinFilm_TechArt}
  PinBase() {reroutePins ["Material (MA)" {expression=&self:MaterialExpressionReroute_0}]}
  PinBase!(Result=@n9.Result) {reroutePins [Result {expression=&self:MaterialExpressionReroute_1}]}
 ]
}
